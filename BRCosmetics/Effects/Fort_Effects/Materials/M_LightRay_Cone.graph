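# Additive, unlit, two-sided material for light ray cones on mesh particles, visible only in evening and night through time-of-day color functions.
# Evening and night color comes from particle color or a constant parameter, multiplied by either vertex alpha or a vertical UV gradient and passed through emissive adjustments.
# Opacity multiplies particle alpha, time-of-day color and vertex alpha or a sharply powered gradient.
# Depth fade softens intersections, except on mobile.
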Material /BRCosmetics/Effects/Fort_Effects/Materials/M_LightRay_Cone
graph {
 properties {
  bUsedWithMeshParticles=true
  blendMode=BLEND_Additive
  pixelDepthOffsetMode=PDOM_Legacy
  shadingModel=MSM_Unlit
  twoSided=true
  usedShadingModels=Unlit
 }
 nodes [
  n1 = Constant()
  n2 = ParticleColor()
  n3 = StaticSwitchParameter(True=VectorParameter() {parameterName=ConstantColor, defaultValue=<rgba 1 1 1 1>}.RGB, False=@n2.RGB) {parameterName=ConstantColor}
  n4 = VertexColor()
  n5 = Multiply(A=@n4.A, B=ScalarParameter() {defaultValue=1, parameterName=EmissiveAmount})
  n6 = MaterialFunctionCall() {materialFunction=&LinearGradient}
  MaterialFunctionCall(Emissive=Multiply(
    A=MaterialFunctionCall(Color_Morning=@n1, Color_Day=@n1, Color_Evening=@n3, Color_Night=@n3) {materialFunction=&Color_Time_of_Day}.Result
    B=StaticSwitchParameter(True=@n5, False=@n6.VGradient) {parameterName=UseVertexColor}
   )) {materialFunction=&MF_FX_EmissiveAdjustments}
  n7 = Power(Base=@n6.VGradient) {constExponent=6}
  n8 = Multiply(
   A=@n2.A
   B=Multiply(
    A=MaterialFunctionCall(Color_Morning=@n1, Color_Day=@n1, Color_Evening=@n3, Color_Night=@n3) {materialFunction=&Color_Time_of_Day}.Result
    B=StaticSwitchParameter(True=@n4.A, False=@n7) {parameterName=UseVertexColor}
   )
  )
  ShadingPathSwitch(Default=DepthFade(Opacity=@n8, FadeDistance=ScalarParameter() {parameterName=FadeDistance}), Mobile=@n8) {desc=Migrated from FeatureLevelSwitch}
 ]
}
