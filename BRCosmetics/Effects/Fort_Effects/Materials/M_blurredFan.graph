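# Translucent, two-sided lit material for mesh particles that fakes spinning, motion-blurred fan blades, using color curves from the SnowNinja glider.
# An absolute sine along U, scrolled by time with slight noise jitter, forms blades whose speed and count come from dynamic parameters.
# V-indexed curve atlas rows set blade width and emissive color, with optional particle tint.
# The blade mask drives depth-faded opacity (unfaded on mobile), and a camera offset scales with V.
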Material /BRCosmetics/Effects/Fort_Effects/Materials/M_blurredFan
graph {
 properties {
  bUsedWithMeshParticles=true
  bUsedWithNiagaraMeshParticles=true
  blendMode=BLEND_Translucent
  pixelDepthOffsetMode=PDOM_Legacy
  translucencyLightingMode=TLM_VolumetricDirectional
  twoSided=true
  usedShadingModels=Default Lit
 }
 nodes [
  n1 = ComponentMask(None=TextureCoordinate()) {g=true}
  n2 = CurveAtlasRowParameter(CurveTime=@n1) {
   atlas=&AthenaCurveAtlas01
   curve=/Game/Gadgets/Assets/VinderTech_GliderChute/Glider_SnowNinja/FX/Curve_FanColor_Emissive
   defaultValue=45
  }
  n3 = CurveAtlasRowParameter(CurveTime=@n1) {
   atlas=&AthenaCurveAtlas01
   curve=/Game/Gadgets/Assets/VinderTech_GliderChute/Glider_SnowNinja/FX/Curve_FanColor
   defaultValue=44
  }
  n4 = Time()
  n5 = Divide(A=@n4) {constB=4}
  n6 = Divide(A=@n5) {constB=64}
  n7 = Multiply(
   A=TextureSample(UVs=AppendVector(A=@n5, B=@n6)) {samplerType=SAMPLERTYPE_Normal, texture=/Game/Global/Textures/Default/Noise/64x64Noise}.R
   B=ScalarParameter() {defaultValue=0.02, parameterName=RandomSpeed}
  ) {constB=0.02}
  n8 = DynamicParameter() {defaultValue=<rgba 1 2 1 4>, paramNames [Speed,FanNum,TimeOffset,CameraOffset]}
  n9 = Add(A=ComponentMask(None=TextureCoordinate()) {r=true}, B=Multiply(A=Add(A=@n4, B=@n7), B=@n8.Speed))
  n10 = Multiply(A=@n8.Speed, B=@n8.Speed)
  n11 = Floor(None=@n8.FanNum)
  n12 = Sine(None=Multiply(A=Add(A=@n9, B=@n10), B=Multiply(A=@n11) {constB=0.5}))
  n13 = Saturate(None=Subtract(A=@n3.A, B=Abs(None=@n12)))
  n14 = Multiply(A=@n2, B=@n13)
  n15 = ParticleColor()
  MaterialFunctionCall(Emissive=LinearInterpolate(A=@n14, B=Multiply(A=@n14, B=@n15.RGB), Alpha=ScalarParameter() {parameterName=UseParticleColor?})) {materialFunction=&MF_FX_EmissiveAdjustments}
  n16 = TextureCoordinate()
  MaterialFunctionCall(Offset Amount=Multiply(A=@n8.CameraOffset, B=ComponentMask(None=@n16) {g=true})) {materialFunction=/Engine/Functions/Engine_MaterialFunctions02/WorldPositionOffset/CameraOffset}
  n17 = TextureCoordinate() {coordinateIndex=1}
  Normalize(VectorInput=AppendVector(A=Subtract(A=@n17) {constB=0.5}, B=Constant() {r=1}))
  Reroute(None=@n3)
  n18 = Multiply(A=LinearInterpolate(Alpha=@n13) {constA=0.2, constB=0.8}, B=ParticleColor().A)
  ShadingPathSwitch(Default=DepthFade(Opacity=@n18, FadeDistance=Constant() {r=10}), Mobile=@n18) {desc=Migrated from FeatureLevelSwitch}
 ]
}
